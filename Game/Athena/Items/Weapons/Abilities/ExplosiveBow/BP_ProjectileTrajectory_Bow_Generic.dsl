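; Projectile trajectory preview for bow weapons, drawing an arc and target marker from supplied spline points.
; BeginPlay builds a spline mesh pool sharing one dynamic material; a looping timer can ask the owner's trajectory interface for updates.
; Updates blend points toward the muzzle, shrink tangents to prevent looping, lay meshes along the spline and set end fade.
; Hit distance from the weapon sets target validity, which OnRep_InvalidTarget shows through material parameters.
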(event EventBeginPlay
  (bind _maxmeshcount (Variables|Default|GetMaxMeshCount))
  (for _array_element (Variables|Trajectory|GetSplineMeshComponents)
    (Rendering|SetOnlyOwnerSee _array_element true :!class /Script/Engine.PrimitiveComponent))
  (Rendering|SetOnlyOwnerSee (Variables|Default|GetTarget) true :!class /Script/Engine.PrimitiveComponent)
  (Rendering|SetVisibility (Variables|Default|GetTarget) false false)
  (Utilities|Array|Resize (Variables|Trajectory|GetSplineMeshComponents) _maxmeshcount)
  (for _index (range _maxmeshcount)
    (bind _returnvalue (AddComponent|NODE_AddSplineMeshComponent-0 :bManualAttachment true))
    (Rendering|Material|SetMaterial _returnvalue 0 (Variables|Default|GetSplineMeshMat))
    (Utilities|Array|SetArrayElem (Variables|Trajectory|GetSplineMeshComponents) _index _returnvalue)
    (if (== _index 0)
      (bind _returnvalue_2 (Rendering|Material|CreateDynamicMaterialInstance _returnvalue 0 (Rendering|Material|GetMaterial _returnvalue 0) "None" :!class /Script/Engine.PrimitiveComponent))
      (Variables|VFX|SetSplineMID _returnvalue_2)
      (bind _returnvalue_3 (Rendering|Material|CreateDynamicMaterialInstance (Variables|Default|GetTarget) 0 (Rendering|Material|GetMaterial (Variables|Default|GetTarget) 0) "None" :!class /Script/Engine.PrimitiveComponent))
      (Variables|VFX|SetTargetMID _returnvalue_3)
      (else
        (Rendering|Material|SetMaterial _returnvalue 0 (Variables|VFX|GetSplineMID))))))

(event Custom|Pulse
  (bind _splinemid (Variables|VFX|GetSplineMID))
  (Rendering|Material|SetScalarParameterValue _splinemid "PulseStart" (Utilities|Time|GetGameTimeinSeconds) :!class /Script/Engine.MaterialInstanceDynamic)
  (Rendering|Material|SetScalarParameterValue _splinemid "OpacityMult" 0.1 :!class /Script/Engine.MaterialInstanceDynamic))

(event Trajectory|EventSetTrajectorySpline
  (params (SplinePoints :type Array%20of%20Vectors) (SplineTangents :type Array%20of%20Vectors))
  (CallFunction|SetTrajectorySplineInternal SplinePoints SplineTangents)
  (if (Variables|Default|GetCheckForValidTarget)
    _))

(event Custom|Check%20Valid%20Target
  (params (Trajectory%20Hit%20Result :type Hit%20Result%20Structure))
  (bind (_bblockinghit _binitialoverlap _time _distance _location _impactpoint) (Collision|BreakHitResult Trajectory%20Hit%20Result))
  (if (Variables|Default|GetCheckForValidTarget)
    (if (and _bblockinghit (>= (Variables|Default|GetProjMaxDistance) (Math|Vector|Distance(Vector) _impactpoint (Transformation|GetActorLocation (Variables|Default|GetOwnerWeapon)))))
      (Variables|Default|SetValidTarget true)
      (else
        (Variables|Default|SetValidTarget false)))))

(fn OnRep_InvalidTarget
  (if (Variables|Default|GetValidTarget)
    (Rendering|Material|SetScalarParameterValue (Variables|VFX|GetSplineMID) "Targetable" 1.0 :!class /Script/Engine.MaterialInstanceDynamic)
    (Rendering|Material|SetScalarParameterValue (Variables|VFX|GetTargetMID) "Targetable" 1.0 :!class /Script/Engine.MaterialInstanceDynamic)
    (else
      (Rendering|Material|SetScalarParameterValue (Variables|VFX|GetSplineMID) "Targetable" 0.0 :!class /Script/Engine.MaterialInstanceDynamic)
      (Rendering|Material|SetScalarParameterValue (Variables|VFX|GetTargetMID) "Targetable" 0.0 :!class /Script/Engine.MaterialInstanceDynamic))))

(fn SetShouldUpdateFromOwner
  (params (ShouldUpdate :type Boolean) (Owner :type Object%20Reference) (Interval :type Float%20%28double-precision%29))
  (if ShouldUpdate
    (Variables|Default|SetTrajectory%20Owner Owner)
    (Utilities|Time|ClearandInvalidateTimerbyHandle (Variables|Default|GetUpdate%20Timer%20Handle))
    (bind _returnvalue (Utilities|Time|SetTimerbyFunctionName 0 "UpdateFromTrajectoryOwner" Interval true))
    (Variables|Default|SetUpdate%20Timer%20Handle _returnvalue)
    (else
      (Variables|Default|SetTrajectory%20Owner 0)
      (Utilities|Time|ClearandInvalidateTimerbyHandle (Variables|Default|GetUpdate%20Timer%20Handle)))))

(fn SetTrajectorySplineInternal
  (params (SplinePoints :type Array%20of%20Vectors) (SplineTangents :type Array%20of%20Vectors))
  (bind _spline (Variables|Default|GetSpline))
  (bind _returnvalue (Spline|GetLocationatSplinePoint _spline 0 "World"))
  (bind _spline_2 (Variables|Default|GetSpline))
  (bind _splinemeshoffsetdistance (Variables|Default|GetSplineMeshOffsetDistance))
  (bind _splinemeshlength (Variables|Default|GetSplineMeshLength))
  (bind _splinemid (Variables|VFX|GetSplineMID))
  (bind _target (Variables|Default|GetTarget))
  (if (> (Utilities|Array|Length SplinePoints) 0)
    (Transformation|SetWorldLocation _target (Utilities|Array|Get(aref) SplinePoints (Utilities|Array|LastIndex SplinePoints)) false true)
    (Rendering|SetVisibility _target true false)
    (Spline|SetSplinePoints (Variables|Default|GetSpline) SplinePoints "World" true)
    (Variables|Default|GetOwnerWeapon
      (:then
        ;[1 1 1] Offset spline points to connect them to the weapon muzzle
        (Variables|Default|SetMuzzleOffset (- _returnvalue (Effects|GetMuzzleLocation (Variables|Default|GetOwnerWeapon) -1)))
        (for (_array_element _array_index) SplinePoints
          (if (< _array_index (- (Utilities|Array|Length SplinePoints) 1))
            (Spline|SetLocationatSplinePoint (Variables|Default|GetSpline) _array_index (- _array_element (* (Variables|Default|GetMuzzleOffset) (Math|Float|Clamp(Float) (/ (Math|Vector|VectorLength (- (Transformation|GetWorldLocation (Variables|Default|GetTarget)) _array_element)) (Math|Vector|VectorLength (- (Transformation|GetWorldLocation (Variables|Default|GetTarget)) _returnvalue)))))) "World" true))))
      (:else
        _))
    ;[0 0 0] For some reason, splines freak out and loop onto themselves when tangent magnitudes are too big. Here we scale them all down to 0.1, because that seems to yield reasonably good results
    (for (_array_element_2 _array_index_2) SplinePoints
      (if (Utilities|Array|IsValidIndex SplinePoints (+ _array_index_2 1))
        (LocalVariables|SetLastPointDist (Math|Vector|VectorLength (- _array_element_2 (Utilities|Array|Get(acopy) SplinePoints (+ _array_index_2 1))))))
      (Spline|SetTangentsatSplinePoint (Variables|Default|GetSpline) _array_index_2 (* (Math|Vector|Normalize (Utilities|Array|Get(acopy) SplineTangents (* _array_index_2 2))) 0.1) (* (Math|Vector|Normalize (Utilities|Array|Get(acopy) SplineTangents (+ (* _array_index_2 2) 1))) 0.1) "World" true))
    ;[1 1 1] Place Spline Meshes Along Spline
    (for (_array_element_3 _array_index_3) (Variables|Trajectory|GetSplineMeshComponents)
      (SplineMesh|SetStartandEnd _array_element_3 (Spline|GetLocationatDistanceAlongSpline _spline_2 (+ (* _array_index_3 _splinemeshlength) _splinemeshoffsetdistance) "World") (* (Math|Vector|Normalize (Spline|GetTangentatDistanceAlongSpline _spline_2 (+ (* _array_index_3 _splinemeshlength) _splinemeshoffsetdistance) "World")) (Variables|Default|GetSplineMeshLength)) (Math|Vector|SelectVector (Spline|GetLocationatDistanceAlongSpline _spline_2 (Spline|GetSplineLength _spline_2) "World") (Spline|GetLocationatDistanceAlongSpline _spline_2 (+ _splinemeshoffsetdistance (* (+ _array_index_3 1) _splinemeshlength)) "World") (> (+ _splinemeshoffsetdistance (* (+ _array_index_3 1) _splinemeshlength)) (Spline|GetSplineLength (Variables|Default|GetSpline)))) (Math|Vector|SelectVector (Math|Vector|Normalize (Spline|GetTangentatDistanceAlongSpline _spline_2 (+ _splinemeshoffsetdistance (* (+ _array_index_3 1) _splinemeshlength)) "World")) (* (Math|Vector|Normalize (Spline|GetTangentatDistanceAlongSpline _spline_2 (+ _splinemeshoffsetdistance (* (+ _array_index_3 1) _splinemeshlength)) "World")) (Variables|Default|GetSplineMeshLength)) (> (+ _splinemeshoffsetdistance (* (+ _array_index_3 1) _splinemeshlength)) (Spline|GetSplineLength (Variables|Default|GetSpline)))) true :!class /Script/Engine.SplineMeshComponent)
      ;[1 1 1] Hide Spline Meshes Once we reach the end of the spline
      (if (< (Math|Vector|VectorLength (- (Spline|GetLocationatDistanceAlongSpline _spline_2 (+ (* _array_index_3 _splinemeshlength) _splinemeshoffsetdistance) "World") (Spline|GetLocationatDistanceAlongSpline _spline_2 (+ _splinemeshoffsetdistance (* (+ _array_index_3 1) _splinemeshlength)) "World"))) 1e+02)
        _
        (elif (> (+ (* _array_index_3 _splinemeshlength) _splinemeshoffsetdistance) (Spline|GetSplineLength (Variables|Default|GetSpline)))
          _
          (else
            (Development|SetHiddeninGame _array_element_3 false false))))
      (Development|SetHiddeninGame _array_element_3 true false))
    ;[1 1 1] Set spline mesh material to fade based on spline length
    (Rendering|Material|SetVectorParameterValue _splinemid "StartLocation" (Math|Conversions|ToLinearColor(Vector) (SplineMesh|GetStartPosition (Utilities|Array|Get(acopy) (Variables|Trajectory|GetSplineMeshComponents)))) :!class /Script/Engine.MaterialInstanceDynamic)
    (Rendering|Material|SetVectorParameterValue _splinemid "EndLocation" (Math|Conversions|ToLinearColor(Vector) (Spline|GetLocationatDistanceAlongSpline (Variables|Default|GetSpline) (Spline|GetSplineLength (Variables|Default|GetSpline)) "World")) :!class /Script/Engine.MaterialInstanceDynamic)
    (Rendering|Material|SetScalarParameterValue _splinemid "EndFadeRadius" (Math|Float|Clamp(Float) (* (Spline|GetSplineLength (Variables|Default|GetSpline)) 0.5) 2e+02 2e+04) :!class /Script/Engine.MaterialInstanceDynamic)))

(fn UpdateFromTrajectoryOwner
  (bind _asfort_trajectory_ability_interface (Utilities|Casting|CastToFortTrajectoryAbilityInterface (Variables|Default|GetTrajectory%20Owner)))
  (TrajectoryInterface|UpdateTrajectory _asfort_trajectory_ability_interface :!class /Script/FortniteGame.FortTrajectoryAbilityInterface))

(fn ConstructionScript)
