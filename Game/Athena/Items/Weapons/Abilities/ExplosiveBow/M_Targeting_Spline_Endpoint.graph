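# Masked unlit material for the explosive bow's targeting spline endpoint, drawing a camera-facing reticle.
# World position offset combines a sprite function, scaled by pixel depth and view height, with a camera offset.
# Targetable blends between no-target (red) and target (blue) textures and colors; at long camera distance, the target reticle blends into a solid round dot.
# A time-of-day collection parameter scales color brightness down to half.
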Material /Game/Athena/Items/Weapons/Abilities/ExplosiveBow/M_Targeting_Spline_Endpoint
graph {
 properties {blendMode=BLEND_Masked, pixelDepthOffsetMode=PDOM_Legacy, shadingModel=MSM_Unlit, usedShadingModels=Unlit}
 nodes [
  n1 = ScalarParameter() {defaultValue=10, parameterName=SpriteScale}
  n2 = ViewSize()
  n3 = Divide(A=PixelDepth().R, B=Multiply(A=ComponentMask(None=@n2) {g=true}, B=ScalarParameter() {defaultValue=0.11, parameterName=ScreenWidth}))
  Add(
   A=MaterialFunctionCall!(Center Location, 0-1 UVs, XY scale=Add(A=@n1, B=@n3), OverridePivot, "Normals (Optional)") {materialFunction=&Sprite}.World position offset
   B=MaterialFunctionCall(Offset Amount=ScalarParameter() {defaultValue=70, parameterName=CameraOffset}) {materialFunction=/Engine/Functions/Engine_MaterialFunctions02/WorldPositionOffset/CameraOffset}.Camera Offset
  )
  n4 = TextureSample() {texture=&/T_Targeting_Endpoint_Reticle_NoTarget}
  n5 = TextureSample() {texture=&/T_Targeting_Endpoint_Reticle}
  n6 = SphereMask(A=TextureCoordinate(), B=Constant() {r=0.5}) {attenuationRadius=0.25, hardnessPercent=70}
  n7 = Distance(A=ObjectPositionWS(), B=CameraPositionWS())
  n8 = ScalarParameter() {defaultValue=-4500, parameterName=DistanceOffset}
  n9 = Clamp(None=Divide(A=Add(A=@n7, B=@n8), B=ScalarParameter() {defaultValue=400, parameterName=BrightenDistance}))
  n10 = ScalarParameter() {defaultValue=1, parameterName=Targetable, sliderMax=1}
  LinearInterpolate(A=@n4.R, B=LinearInterpolate(A=@n5.R, B=Multiply(A=@n6) {constB=5}, Alpha=@n9), Alpha=@n10)
  n11 = CollectionParameter() {
   collection=&FortniteMaterialParameters
   parameterName=FortniteActiveTimeOfDay
  }
  n12 = Constant3Vector() {constant=<rgba 0.05 0 0 0>}
  n13 = Constant3Vector() {constant=<rgba 0.015 0.02 0.05 0>}
  n14 = Constant3Vector() {constant=<rgba 1 0 0 0>}
  n15 = Constant3Vector() {constant=<rgba 0.0156 0.7617 3 0>}
  n16 = LinearInterpolate(A=@n5.G, B=@n6, Alpha=@n9)
  Multiply(
   A=LinearInterpolate(Alpha=ComponentMask(None=@n11) {a=true}) {constA=1, constB=0.5}
   B=LinearInterpolate(
    A=LinearInterpolate(A=@n12, B=@n13, Alpha=@n10)
    B=LinearInterpolate(A=@n14, B=@n15, Alpha=@n10)
    Alpha=LinearInterpolate(A=@n4.G, B=@n16, Alpha=@n10)
   )
  )
 ]
}
