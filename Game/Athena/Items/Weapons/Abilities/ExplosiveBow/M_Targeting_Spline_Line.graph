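# Translucent, unlit spline-mesh material for the Explosive Bow's targeting line, drawn without depth testing and colored red or blue by the Targetable parameter.
# A soft cross-line mask blends dark edges into a bright core, adding a timed flash and a looping traveling pulse, scaled by time of day.
# Opacity fades near the start, at far range and by scene depth off mobile, clipping beyond the end's camera distance.
# SplineThicken widens it with pixel depth to hold screen width.
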Material /Game/Athena/Items/Weapons/Abilities/ExplosiveBow/M_Targeting_Spline_Line
graph {
 properties {
  bDisableDepthTest=true
  bUsedWithSplineMeshes=true
  blendMode=BLEND_Translucent
  pixelDepthOffsetMode=PDOM_Legacy
  shadingModel=MSM_Unlit
  twoSided=true
  usedShadingModels=Unlit
 }
 nodes [
  n1 = CollectionParameter() {
   collection=&FortniteMaterialParameters
   parameterName=FortniteActiveTimeOfDay
  }
  n2 = ScalarParameter() {defaultValue=1, parameterName=Targetable, sliderMax=1}
  n3 = LinearInterpolate(A=Constant3Vector() {constant=<rgba 0.05 0 0 0>}, B=Constant3Vector() {constant=<rgba 0.015 0.0208 0.05 0>}, Alpha=@n2)
  n4 = Divide(A=Subtract(A=Time(), B=ScalarParameter() {parameterName=PulseStart}), B=ScalarParameter() {defaultValue=0.25, parameterName=PulseDuration})
  n5 = Power(Base=OneMinus(None=Clamp(None=@n4)))
  n6 = LinearInterpolate(A=Constant3Vector() {constant=<rgba 1 0 0 0>}, B=Constant3Vector() {constant=<rgba 0.0156 0.7617 3 0>}, Alpha=@n2)
  n7 = Add(A=Multiply(A=Multiply(A=@n5) {constB=4}, B=@n6), B=@n6)
  n8 = Constant() {r=0.5}
  n9 = SphereMask(
   A=ComponentMask(None=TextureCoordinate()) {r=true}
   B=@n8
   Radius=@n8
   Hardness=ScalarParameter() {defaultValue=0.25, parameterName=Hardness}
  )
  n10 = Clamp(None=Multiply(A=@n9, B=@n9))
  n11 = Constant() {r=1}
  n12 = ScalarParameter() {defaultValue=12, parameterName=PulseLength}
  n13 = ViewSize()
  n14 = Divide(A=PixelDepth().R, B=Multiply(A=ComponentMask(None=@n13) {g=true}, B=ScalarParameter() {defaultValue=0.65, parameterName=ScreenWidth}))
  n15 = Divide(A=@n12, B=@n14)
  n16 = LinearInterpolate(A=Multiply(A=@n12) {constB=0.1}, Alpha=Time() {bOverride_Period=true, period=1}) {constA=1, constB=0}
  n17 = SphereMask(
   A=Multiply(A=TextureCoordinate(), B=AppendVector(A=@n11, B=@n15))
   B=AppendVector(A=Constant() {r=0.5}, B=Multiply(A=@n15, B=@n16))
   Radius=Constant() {r=0.5}
  ) {hardnessPercent=0}
  n18 = ScalarParameter() {defaultValue=3000, parameterName=PulseFadeDistance}
  n19 = Multiply(
   A=Multiply(A=Power(Base=@n17), B=ScalarParameter() {defaultValue=16, parameterName=PulseBrightness}) {constB=20}
   B=OneMinus(None=MaterialFunctionCall(Distance=@n18) {materialFunction=&Near_Camera_Fade}.Result)
  )
  Multiply(
   A=LinearInterpolate(Alpha=ComponentMask(None=@n1) {a=true}) {constA=1, constB=0.5}
   B=Add(A=LinearInterpolate(A=@n3, B=@n7, Alpha=@n10), B=Multiply(A=@n6, B=@n19))
  )
  n20 = ScalarParameter() {defaultValue=1, parameterName=OpacityMult}
  n21 = WorldPosition()
  n22 = CameraPositionWS()
  n23 = VectorParameter() {parameterName=EndLocation, defaultValue=<rgba 200 0 0 0>}
  n24 = WorldPosition()
  n25 = VectorParameter() {parameterName=StartLocation}
  n26 = ScalarParameter() {defaultValue=1000, parameterName=StartFadeDistance}
  n27 = Multiply(
   A=If(A=Distance(A=@n21.XYZ, B=@n22), B=Distance(A=@n22, B=@n23.RGB), "A > B"=Constant(), "A < B"=Constant() {r=1})
   B=OneMinus(None=SphereMask(A=@n24.XYZ, B=@n25.RGB, Radius=@n26) {hardnessPercent=0})
  )
  n28 = SphereMask(A=@n24.XYZ, B=@n25.RGB, Radius=ScalarParameter() {defaultValue=10000, parameterName=FarFadeDistance}) {hardnessPercent=0}
  n29 = ScalarParameter() {defaultValue=100, parameterName=FadeDistance}
  Multiply(
   A=Multiply(A=Multiply(A=@n9, B=@n20), B=Multiply(A=@n27, B=@n28))
   B=ShadingPathSwitch(Default=DepthFade(FadeDistance=@n29), Mobile=Constant() {r=1}) {desc=Migrated from FeatureLevelSwitch}
  ) {constA=1}
  n30 = Add(A=ScalarParameter() {defaultValue=6, parameterName=Width}, B=@n14)
  Reroute(None=MaterialFunctionCall(WidthBase=@n30, WidthTip=@n30) {materialFunction=&SplineThicken}.WorldPositionOffset)
  VectorParameter() {parameterName=BrightColor, defaultValue=<rgba 0.0104 0.5078 2 0>}
  VectorParameter() {parameterName=EdgeColor, defaultValue=<rgba 0.025 0.0291 0.05 0>}
 ]
}
